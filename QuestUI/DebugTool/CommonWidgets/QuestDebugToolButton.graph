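# QuestDebugToolButton is a CommonButtonBase widget that serves as an icon button for the quest debug tool UI.
# Its layout is an Overlay holding a centred, padded Image_Icon.
# PreConstruct passes the IconBrush variable, which defaults to a small right-arrow texture, to SetIcon.
# SetIcon stores the given brush in IconBrush and applies it to the image, so the icon can also be changed through that function.
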WidgetBlueprint /QuestUI/DebugTool/CommonWidgets/QuestDebugToolButton
graphs [
 EventGraph {nodes [
   Event PreConstruct()
   > SetIcon!(execute, self, IconBrush:Slate Brush Structure=Get IconBrush() -> (IconBrush:Slate Brush Structure)) -> (then)
  ]}
 SetIcon {nodes [
   n1 = SetIcon!() -> (then, IconBrush:Slate Brush Structure) {is=title K2Node_FunctionEntry}
   > n2 = Set IconBrush(IconBrush:Slate Brush Structure=@n1) -> (Output_Get:Slate Brush Structure)
   > SetBrush(self=Get Image_Icon() -> (Image_Icon:Image Object Reference), InBrush=@n2)
  ]}
]
variables [IconBrush {
  type=SlateBrush
  default {
   bIsDynamicallyLoaded=false
   drawAs=Image
   imageType=FullColor
   mirroring=NoMirror
   resourceName=None
   resourceObject=/Game/UI/Style/Button-BasicArrowRight-_normal_
   tiling=NoTile
   imageSize=<16 16>
   margin=<ltrb 0 0 0 0>
   outlineSettings {
    bUseBrushTransparency=false
    roundingType=HalfHeightRadius
    width=0
    color {colorUseRule=UseColor_Specified, specifiedColor=<rgba 0 0 0 0>}
    cornerRadii=<0 0 0 1>
   }
   tintColor {colorUseRule=UseColor_Specified, specifiedColor=<rgba 1 1 1 1>}
   uVRegion {bIsValid=false, max=<0 0>, min=<0 0>}
  }
 }]
widgets {
 info {inheritedWidgetCount=1, parentClass=&CommonButtonBase, rootWidgetClass=&Overlay, widgetCount=3}
 widgets [
  Overlay_22 {widgetClassPath=&Overlay, widget=Overlay_22}
  Image_Icon {
   widgetClassPath=&Image
   bIsVariable=true
   parent=Overlay_22
   widget=Image_Icon
   slot {class=OverlaySlot, properties {horizontalAlignment=HAlign_Center, verticalAlignment=VAlign_Center, padding=<ltrb 4 4 4 4>}}
  }
  InputActionWidget {widgetClassPath=&CommonActionWidget, bInherited=true, widget=null}
 ]
}
