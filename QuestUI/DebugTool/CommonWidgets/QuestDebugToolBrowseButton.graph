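# Quest debug tool widget: a single magnifying-glass button that jumps to an asset in the Content Browser.
# On click, it checks that LocalAsset is valid and passes it to BrowseToAsset from the QuestPanelDebugToolWidgetBase parent; nothing in this graph sets LocalAsset.
# GetBrowseToAssetVisibility returns Visible only when its Asset argument is valid and the game runs in editor or PIE, otherwise Collapsed.
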WidgetBlueprint /QuestUI/DebugTool/CommonWidgets/QuestDebugToolBrowseButton
graphs [
 EventGraph {nodes [
   "On Clicked (Button_BrowseToAsset)"!() -> (OutputDelegate, then, Button:Common Button Base Object Reference) {is=title K2Node_ComponentBoundEvent}
   n1 = Get LocalAsset() -> (LocalAsset:Object Reference)
   > Is Valid(InputObject=@n1)
   > BrowseToAsset!(execute, self:Quest Panel Debug Tool Widget Base Object Reference, Object:Object Reference=@n1) -> (then)
  ]}
 GetBrowseToAssetVisibility {nodes [
   n1 = GetBrowseToAssetVisibility!() -> (then, Asset:Object Reference) {is=title K2Node_FunctionEntry}
   n2 = IsValid(Object=@n1)
   n3 = InEditorOrPIE!() -> (ReturnValue:bool)
   > Return Node!(execute, Visibility:ESlateVisibility Enum=Select(Option 0=Collapsed, Option 1=Visible, Index=AND Boolean(A=@n3, B=@n2)))
  ]}
]
variables [LocalAsset {type=Object, default=null}]
widgets {
 info {
  inheritedWidgetCount=0
  parentClass=/Script/QuestUI.QuestPanelDebugToolWidgetBase
  rootWidgetClass=&/QuestDebugToolButton.QuestDebugToolButton_C
  widgetCount=1
 }
 widgets [Button_BrowseToAsset {
   widgetClassPath=&/QuestDebugToolButton
   bIsVariable=true
   widget=Button_BrowseToAsset
   properties {
    toolTipText=Jump to this asset in the Content Browser
    iconBrush {
     bIsDynamicallyLoaded=false
     drawAs=Image
     imageType=FullColor
     mirroring=NoMirror
     resourceName=None
     resourceObject=/Game/UI/Foundation/Textures/Icons/HUD/T-IconMagnifying-Glass-32
     tiling=NoTile
     imageSize=<24 24>
     margin=<ltrb 0 0 0 0>
     outlineSettings {
      bUseBrushTransparency=false
      roundingType=HalfHeightRadius
      width=0
      color {colorUseRule=UseColor_Specified, specifiedColor=<rgba 0 0 0 0>}
      cornerRadii=<0 0 0 1>
     }
     tintColor {colorUseRule=UseColor_Specified, specifiedColor=<rgba 1 1 1 1>}
     uVRegion {bIsValid=false, max=<0 0>, min=<0 0>}
    }
   }
  }]
}
