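# Labeled checkbox widget for the quest debug tool UI, built on CommonButtonBase.
# The CheckBox is hit-test invisible, so the button's click handler toggles it (undetermined becomes checked) and broadcasts OnStateChanged with the new bool.
# SetLabel stores and displays the label text, and PreConstruct applies it from the Label variable.
# SetState sets the checked state without calling OnStateChanged.
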WidgetBlueprint /QuestUI/DebugTool/CommonWidgets/QuestDebugToolCheckbox
graphs [
 EventGraph {nodes [
   Event On Clicked()
   n1 = Get CheckBox() -> (CheckBox:Check Box Object Reference)
   > SetCheckedState!(
    execute
    self:Check Box Object Reference=@n1
    InCheckedState:ECheckBoxState Enum=Select!(
     Unchecked:ECheckBoxState Enum=Checked
     Checked:ECheckBoxState Enum=Unchecked
     Undetermined:ECheckBoxState Enum=Checked
     Index:ECheckBoxState Enum=GetCheckedState!(self:Check Box Object Reference=@n1) -> (ReturnValue:ECheckBoxState Enum=Unchecked)
    ) -> (ReturnValue:ECheckBoxState Enum)
   ) -> (then)
   > Call On State Changed!(execute:Exec, self:Quest Debug Tool Checkbox Object Reference, State:bool=IsChecked(self=Get CheckBox())) -> (then:Exec) {is=title K2Node_CallDelegate}
   Event PreConstruct()
   > SetLabel!(execute, self, Label:Text=Get Label() -> (Label:Text)) -> (then)
  ]}
 OnStateChanged {nodes [OnStateChanged!() -> (then, State:bool=~) {is=title K2Node_FunctionEntry}]}
 SetLabel {nodes [
   n1 = SetLabel!() -> (then, Label:Text) {is=title K2Node_FunctionEntry}
   > n2 = Set Label(Label:Text=@n1) -> (Output_Get:Text)
   > "SetText (Text)"(self=Get Text_Label() -> (Text_Label:Common Text Object Reference), InText=@n2)
  ]}
 SetState {nodes [
   n1 = SetState!() -> (then, State:bool=~) {is=title K2Node_FunctionEntry}
   > SetIsChecked!(execute, self:Check Box Object Reference=Get CheckBox(), InIsChecked:bool=@n1) -> (then)
  ]}
]
variables [
 Label {type=string, default=Label}
 OnStateChanged {}
]
widgets {
 info {inheritedWidgetCount=1, parentClass=&CommonButtonBase, rootWidgetClass=&HorizontalBox, widgetCount=4}
 widgets [
  HorizontalBox_20 {widgetClassPath=&HorizontalBox, widget=HorizontalBox_20}
  CheckBox {
   widgetClassPath=/Script/UMG.CheckBox
   bIsVariable=true
   parent=HorizontalBox_20
   widget=CheckBox
   properties {visibility=HitTestInvisible}
   slot {class=HorizontalBoxSlot, properties {verticalAlignment=VAlign_Center}}
  }
  Text_Label {
   widgetClassPath=&CommonTextBlock
   bIsVariable=true
   parent=HorizontalBox_20
   widget=Text_Label
   properties {
    scrollStyle=&TextScrollStyle-Base_C
    style=/QuestUI/DebugTool/TextStyles/TextStyle_DebugTool_Body.TextStyle_DebugTool_Body_C
    text=Label
    colorAndOpacity {colorUseRule=UseColor_Specified, specifiedColor=<rgba 1 0.9301 0 1>}
    font {
     bForceMonospaced=false
     bMaterialIsStencil=false
     fontMaterial=null
     fontObject=&NotoSans
     letterSpacing=0
     monospacedWidth=1
     size=12.75
     skewAmount=0
     typefaceFontName=NotoSans-Bold
     outlineSettings {
      bApplyOutlineToDropShadows=false
      bMiteredCorners=false
      bSeparateFillAlpha=false
      outlineMaterial=null
      outlineSize=0
      outlineColor=<rgba 0 0 0 1>
     }
    }
    shadowOffset=<0 0>
   }
   slot {class=HorizontalBoxSlot, properties {verticalAlignment=VAlign_Center}}
  }
  InputActionWidget {widgetClassPath=&CommonActionWidget, bInherited=true, widget=null}
 ]
}
