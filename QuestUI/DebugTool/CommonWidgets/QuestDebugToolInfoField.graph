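# Reusable Quest debug tool UI widget that shows a header label above a data value on a tinted background.
# PreConstruct applies the Header and Data texts, and Multiline toggles between auto-wrapped and scrolling data text.
# SetHeader and SetData store and display new text.
# SetDisplayAsAlert switches the data text between yellow and red and shows or collapses the alert icon.
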WidgetBlueprint /QuestUI/DebugTool/CommonWidgets/QuestDebugToolInfoField
graphs [
 EventGraph {nodes [
   Event PreConstruct()
   > "SetText (Text)"(self=Get Text_Header() -> (Text_Header:Common Text Object Reference), InText=Get Header() -> (Header:Text))
   > "SetText (Text)"(self=Get Text_Data() -> (Text_Data:Common Text Object Reference), InText=Get Data() -> (Data:Text))
   n1 = Get Multiline() -> (Multiline:bool)
   n2 = Get Text_Data()
   > SetScrollingEnabled!(execute, self:Common Text Object Reference=@n2, bInIsScrollingEnabled:bool=NOT Boolean(A=@n1)) -> (then)
   > SetAutoWrapText!(execute, self:Text Object Reference=@n2, InAutoTextWrap:bool=@n1) -> (then)
   > SetDisplayAsAlert!(execute, self, InDisplayAsAlert:bool=Get DisplayAsAlert() -> (DisplayAsAlert:bool)) -> (then)
  ]}
 SetData {nodes [
   n1 = SetData!() -> (then, Data:Text) {is=title K2Node_FunctionEntry}
   > n2 = Set Data(Data:Text=@n1) -> (Output_Get:Text)
   > "SetText (Text)"(self=Get Text_Data(), InText=@n2)
  ]}
 SetDisplayAsAlert {nodes [
   n1 = SetDisplayAsAlert!() -> (then, InDisplayAsAlert:bool=~) {is=title K2Node_FunctionEntry}
   > n2 = Set DisplayAsAlert(DisplayAsAlert:bool=@n1) -> (Output_Get:bool)
   > SetColorAndOpacity(
    self=Get Text_Data()
    InColorAndOpacity=Make Slate Color(SpecifiedColor=SelectColor(A=(R=1,G=0,B=0,A=1), B=(R=1,G=0.7991,B=0,A=1), bPickA=@n2))
   )
   > SetVisibility(
    self=Get Icon_Alert() -> (Icon_Alert:Image Object Reference)
    InVisibility=Select(Option 0=Collapsed, Option 1=HitTestInvisible, Index=Get DisplayAsAlert())
   )
  ]}
 SetHeader {nodes [
   n1 = SetHeader!() -> (then, Header:Text) {is=title K2Node_FunctionEntry}
   > n2 = Set Header(Header:Text=@n1) -> (Output_Get:Text)
   > "SetText (Text)"(self=Get Text_Header(), InText=@n2)
  ]}
]
variables [
 Data {type=string, default=Data}
 DisplayAsAlert {type=boolean}
 Header {type=string, default=Header}
 Multiline {type=boolean}
]
widgets {
 info {inheritedWidgetCount=0, parentClass=&UserWidget, rootWidgetClass=&Overlay, widgetCount=8}
 widgets [
  Overlay_0 {widgetClassPath=&Overlay, widget=Overlay_0}
  BG {
   widgetClassPath=&Image
   parent=Overlay_0
   widget=BG
   properties {colorAndOpacity=<rgba 0 0.3184 0.3646 0.1>}
   slot {class=OverlaySlot, properties {horizontalAlignment=HAlign_Fill, verticalAlignment=VAlign_Fill, padding=<ltrb 3 3 3 3>}}
  }
  VerticalBox_28 {
   widgetClassPath=&VerticalBox
   parent=Overlay_0
   widget=VerticalBox_28
   slot {class=OverlaySlot, properties {horizontalAlignment=HAlign_Fill, verticalAlignment=VAlign_Fill, padding=<ltrb 5 5 5 5>}}
  }
  ScaleBox_0 {widgetClassPath=&ScaleBox, parent=VerticalBox_28, widget=ScaleBox_0, slot {class=VerticalBoxSlot}}
  Text_Header {
   widgetClassPath=&CommonTextBlock
   bIsVariable=true
   parent=ScaleBox_0
   widget=Text_Header
   properties {
    style=/QuestUI/DebugTool/TextStyles/TextStyle_DebugTool_Header.TextStyle_DebugTool_Header_C
    text=Header
    font {
     bForceMonospaced=false
     bMaterialIsStencil=false
     fontMaterial=null
     fontObject=&NotoSans
     letterSpacing=0
     monospacedWidth=1
     size=9
     skewAmount=0
     typefaceFontName=NotoSans-Bold
     outlineSettings {
      bApplyOutlineToDropShadows=false
      bMiteredCorners=false
      bSeparateFillAlpha=false
      outlineMaterial=null
      outlineSize=0
      outlineColor=<rgba 0 0 0 1>
     }
    }
    shadowOffset=<0 0>
   }
   slot {class=ScaleBoxSlot, properties {horizontalAlignment=HAlign_Fill, verticalAlignment=VAlign_Fill}}
  }
  HorizontalBox_0 {
   widgetClassPath=&HorizontalBox
   parent=VerticalBox_28
   widget=HorizontalBox_0
   slot {class=VerticalBoxSlot, properties {size {sizeRule=Fill, value=1}}}
  }
  Icon_Alert {
   widgetClassPath=&Image
   bIsVariable=true
   parent=HorizontalBox_0
   widget=Icon_Alert
   properties {
    brush {
     bIsDynamicallyLoaded=false
     drawAs=Image
     imageType=FullColor
     mirroring=NoMirror
     resourceName=None
     resourceObject=/FortUILibrary/Textures/MaterialInstances/MI_UIKit_Icon_Alert
     tiling=NoTile
     imageSize=<16 16>
     margin=<ltrb 0 0 0 0>
     outlineSettings {
      bUseBrushTransparency=false
      roundingType=HalfHeightRadius
      width=0
      color {colorUseRule=UseColor_Specified, specifiedColor=<rgba 0 0 0 0>}
      cornerRadii=<0 0 0 0>
     }
     tintColor {colorUseRule=UseColor_Specified, specifiedColor=<rgba 1 1 1 1>}
     uVRegion {bIsValid=false, max=<0 0>, min=<0 0>}
    }
    colorAndOpacity=<rgba 1 0 0 1>
   }
   slot {class=HorizontalBoxSlot, properties {horizontalAlignment=HAlign_Left, verticalAlignment=VAlign_Center, padding=<ltrb 0 0 4 0>}}
  }
  Text_Data {
   widgetClassPath=&CommonTextBlock
   bIsVariable=true
   parent=HorizontalBox_0
   widget=Text_Data
   properties {
    scrollStyle=&TextScrollStyle-Base_C
    style=/QuestUI/DebugTool/TextStyles/TextStyle_DebugTool_Body.TextStyle_DebugTool_Body_C
    text=Data
    colorAndOpacity {colorUseRule=UseColor_Specified, specifiedColor=<rgba 1 0.9301 0 1>}
    font {
     bForceMonospaced=false
     bMaterialIsStencil=false
     fontMaterial=null
     fontObject=&NotoSans
     letterSpacing=0
     monospacedWidth=1
     size=12.75
     skewAmount=0
     typefaceFontName=NotoSans-Bold
     outlineSettings {
      bApplyOutlineToDropShadows=false
      bMiteredCorners=false
      bSeparateFillAlpha=false
      outlineMaterial=null
      outlineSize=0
      outlineColor=<rgba 0 0 0 1>
     }
    }
    shadowOffset=<0 0>
   }
   slot {class=HorizontalBoxSlot, properties {size {sizeRule=Fill, value=1}}}
  }
 ]
}
