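# QuestDebugToolOrganizationFilters is a Quest debug tool widget that presents a height-capped list for filtering by gameplay tags.
# Init hands a root organization node and the checked tag set to ListView_Tags, which displays them with organization entry widgets.
# Construct binds the list's On Checked Tags Changed event and Destruct unbinds it.
# That handler re-broadcasts the tags through the widget's OnActiveTagsChanged dispatcher for outside listeners.
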WidgetBlueprint /QuestUI/DebugTool/CommonWidgets/QuestDebugToolOrganizationFilters
graphs [
 EventGraph {nodes [
   n1 = HandleCheckedTagsChanged!() -> (OutputDelegate, then, ActiveTags:"Gameplay Tag Container Structure (by ref)") {is=title K2Node_CustomEvent}
   > Call On Active Tags Changed!(
    execute:Exec
    self:Quest Debug Tool Organization Filters Object Reference
    ActiveTags:"Gameplay Tag Container Structure (by ref)"=@n1.ActiveTags
   ) -> (then:Exec) {is=title K2Node_CallDelegate}
   Event Construct()
   > Bind Event to On Checked Tags Changed(
    self:Quest Panel Debug Tool Organization List View Object Reference=Get ListView_Tags() -> (ListView_Tags:Quest Panel Debug Tool Organization List View Object Reference)
    Delegate=@n1
   ) {is=title K2Node_AddDelegate}
   Event Destruct()
   > Unbind Event from On Checked Tags Changed(self:Quest Panel Debug Tool Organization List View Object Reference=Get ListView_Tags(), Delegate=@n1) {is=title K2Node_RemoveDelegate}
  ]}
 Init {nodes [
   n1 = Init!() -> (then, RootNode:Quest Panel Debug Organization Node Object Reference, AllCheckedTags:"Gameplay Tag Container Structure (by ref)") {is=title K2Node_FunctionEntry}
   > Init!(
    execute
    self:Quest Panel Debug Tool Organization List View Object Reference=Get ListView_Tags()
    InRootNode:Quest Panel Debug Organization Node Object Reference=@n1
    AllCheckedTags:"Gameplay Tag Container Structure (by ref)"=@n1.AllCheckedTags
   ) -> (then) {is=title K2Node_CallFunction}
  ]}
 OnActiveTagsChanged {nodes [OnActiveTagsChanged!() -> (then, ActiveTags:"Gameplay Tag Container Structure (by ref)") {is=title K2Node_FunctionEntry}]}
]
variables [OnActiveTagsChanged {}]
widgets {
 info {inheritedWidgetCount=0, parentClass=&UserWidget, rootWidgetClass=&SizeBox, widgetCount=2}
 widgets [
  SizeBox_0 {widgetClassPath=&SizeBox, widget=SizeBox_0, properties {bOverride_MaxDesiredHeight=true, maxDesiredHeight=300}}
  ListView_Tags {
   widgetClassPath=/Script/QuestUI.QuestPanelDebugToolOrganizationListView
   bIsVariable=true
   parent=SizeBox_0
   widget=ListView_Tags
   properties {
    entryWidgetClass=&/QuestDebugToolOrganizationEntry.QuestDebugToolOrganizationEntry_C
    numDesignerPreviewEntries=3
   }
   slot {class=SizeBoxSlot}
  }
 ]
}
