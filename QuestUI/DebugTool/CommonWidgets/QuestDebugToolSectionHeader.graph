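# QuestDebugToolSectionHeader is a reusable section header widget for the Quest UI debug tool.
# An overlay stacks a translucent teal background image behind a vertical box, whose scale box holds a bold, dark CommonTextBlock.
# PreConstruct copies the Header variable into that text block, so the label is set from a single property.
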WidgetBlueprint /QuestUI/DebugTool/CommonWidgets/QuestDebugToolSectionHeader
graphs [EventGraph {nodes [
   Event PreConstruct()
   > "SetText (Text)"(self=Get Text_Header() -> (Text_Header:Common Text Object Reference), InText=Get Header() -> (Header:Text))
  ]}]
variables [Header {type=string, default=Header}]
widgets {
 info {inheritedWidgetCount=0, parentClass=&UserWidget, rootWidgetClass=&Overlay, widgetCount=5}
 widgets [
  Overlay_0 {widgetClassPath=&Overlay, widget=Overlay_0}
  BG {
   widgetClassPath=&Image
   parent=Overlay_0
   widget=BG
   properties {colorAndOpacity=<rgba 0 0.3184 0.3646 0.5>}
   slot {class=OverlaySlot, properties {horizontalAlignment=HAlign_Fill, verticalAlignment=VAlign_Fill, padding=<ltrb 5 5 5 5>}}
  }
  VerticalBox_28 {
   widgetClassPath=&VerticalBox
   parent=Overlay_0
   widget=VerticalBox_28
   slot {class=OverlaySlot, properties {horizontalAlignment=HAlign_Fill, verticalAlignment=VAlign_Fill, padding=<ltrb 10 10 10 10>}}
  }
  ScaleBox_0 {widgetClassPath=&ScaleBox, parent=VerticalBox_28, widget=ScaleBox_0, slot {class=VerticalBoxSlot}}
  Text_Header {
   widgetClassPath=&CommonTextBlock
   bIsVariable=true
   parent=ScaleBox_0
   widget=Text_Header
   properties {
    style=/FortUILibrary/Styles/Typography/Foundation/TextStyle-Base-S-B-Blue.TextStyle-Base-S-B-Blue_C
    text=Header
    colorAndOpacity {colorUseRule=UseColor_Specified, specifiedColor=<rgba 0.008 0.011 0.0262 1>}
    font {
     bForceMonospaced=false
     bMaterialIsStencil=false
     fontMaterial=null
     fontObject=&NotoSans
     letterSpacing=0
     monospacedWidth=1
     size=18
     skewAmount=0
     typefaceFontName=NotoSans-Bold
     outlineSettings {
      bApplyOutlineToDropShadows=false
      bMiteredCorners=false
      bSeparateFillAlpha=false
      outlineMaterial=null
      outlineSize=0
      outlineColor=<rgba 0 0 0 1>
     }
    }
    shadowOffset=<0 0>
   }
   slot {class=ScaleBoxSlot, properties {horizontalAlignment=HAlign_Fill, verticalAlignment=VAlign_Fill}}
  }
 ]
}
