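# Arrow-icon button widget for the Quest UI debug tool that selects a list entry in the debugger.
# Clicking it passes its LocalListItemId to FindAndSelectEntry on itself, a QuestPanelDebugToolWidgetBase subclass.
# GetSelectEntryVisibility returns Hidden when the local ID is empty or matches the supplied active item ID, ignoring case, and Visible otherwise.
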WidgetBlueprint /QuestUI/DebugTool/CommonWidgets/QuestDebugToolSelectEntryButton
graphs [
 EventGraph {nodes [
   "On Clicked (Button_SelectEntry)"!() -> (OutputDelegate, then, Button:Common Button Base Object Reference) {is=title K2Node_ComponentBoundEvent}
   > FindAndSelectEntry!(
    execute
    self:Quest Panel Debug Tool Widget Base Object Reference
    PrimaryAssetTemplateId:String=Get LocalListItemId() -> (LocalListItemId:String)
   ) -> (then)
  ]}
 GetSelectEntryVisibility {nodes [
   n1 = GetSelectEntryVisibility!() -> (then, ActiveListItemId:"String (by ref)", LocalListItemId:"String (by ref)") {is=title K2Node_FunctionEntry}
   n2 = "Equal, Case Insensitive (String)"(A=@n1, B=@n1.LocalListItemId)
   n3 = IsEmpty(InString=@n1.LocalListItemId)
   > Return Node!(execute, Visibility:ESlateVisibility Enum=Select(Option 0=Visible, Option 1=Hidden, Index=OR Boolean(A=@n3, B=@n2)))
  ]}
]
variables [LocalListItemId {type=string, default=None}]
widgets {
 info {
  inheritedWidgetCount=0
  parentClass=/Script/QuestUI.QuestPanelDebugToolWidgetBase
  rootWidgetClass=&/QuestDebugToolButton.QuestDebugToolButton_C
  widgetCount=1
 }
 widgets [Button_SelectEntry {
   widgetClassPath=&/QuestDebugToolButton
   bIsVariable=true
   widget=Button_SelectEntry
   properties {
    toolTipText=Select this entry in the Quest UI Debugger
    iconBrush {
     bIsDynamicallyLoaded=false
     drawAs=Image
     imageType=FullColor
     mirroring=NoMirror
     resourceName=None
     resourceObject=/Game/UI/Style/Button-BasicArrowRight-_normal_
     tiling=NoTile
     imageSize=<24 24>
     margin=<ltrb 0 0 0 0>
     outlineSettings {
      bUseBrushTransparency=false
      roundingType=HalfHeightRadius
      width=0
      color {colorUseRule=UseColor_Specified, specifiedColor=<rgba 0 0 0 0>}
      cornerRadii=<0 0 0 1>
     }
     tintColor {colorUseRule=UseColor_Specified, specifiedColor=<rgba 1 1 1 1>}
     uVRegion {bIsValid=false, max=<0 0>, min=<0 0>}
    }
   }
  }]
}
